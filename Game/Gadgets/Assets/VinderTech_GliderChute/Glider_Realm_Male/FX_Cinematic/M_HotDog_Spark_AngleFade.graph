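# Additive unlit spark particle material for the VinderTech glider's cinematic effects, with optional view-angle fade.
# Particle size and camera-to-direction angle remap UVs so a radial gradient stretches into a round-capped "hotdog" streak, except on mobile.
# Particle color tints the gradient, with switchable desaturation and tip fade, plus blendable time-of-day tint and emissive adjustments.
# Opacity math adds near-camera fade, plus depth fade on the default shading path.
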Material /Game/Gadgets/Assets/VinderTech_GliderChute/Glider_Realm_Male/FX_Cinematic/M_HotDog_Spark_AngleFade
graph {
 properties {
  bEnableResponsiveAA=true
  bUsedWithMeshParticles=true
  bUsedWithNiagaraMeshParticles=true
  bUsedWithNiagaraSprites=true
  bUsedWithParticleSprites=true
  blendMode=BLEND_Additive
  pixelDepthOffsetMode=PDOM_Legacy
  shadingModel=MSM_Unlit
  usedShadingModels=Unlit
 }
 nodes [
  n1 = ParticleSize()
  n2 = MaterialFunctionCall(Float2=@n1) {materialFunction=&BreakOutFloat2Components}
  n3 = CameraVectorWS()
  n4 = ParticleDirection()
  n5 = LinearInterpolate(B=Constant2Vector() {g=0.1, r=1}, Alpha=Abs(None=DotProduct(A=@n3, B=@n4))) {constA=1}
  n6 = Constant() {r=1}
  n7 = Normalize(VectorInput=Multiply(A=If(A=@n2.R, B=@n2.R, "A > B"=@n5, "A == B"=@n6, "A < B"=@n5), B=@n1))
  n8 = ComponentMask(None=@n7) {r=true}
  n9 = Clamp(None=Divide(A=ComponentMask(None=@n7) {g=true}, B=@n8), Min=@n8) {clampMode=CMODE_ClampMin}
  n10 = Multiply(A=AppendVector(A=Constant() {r=1}, B=@n9), B=TextureCoordinate())
  n11 = ComponentMask(None=Clamp(None=@n10)) {g=true}
  n12 = LinearInterpolate(A=@n10, B=Add(A=TextureCoordinate() {vTiling=0}, B=Constant2Vector() {g=0.5}), Alpha=@n11) {constAlpha=0, desc=Blend top and middle}
  n13 = Add(A=@n10, B=AppendVector(A=Constant(), B=OneMinus(None=@n9)))
  n14 = ComponentMask(None=Clamp(None=@n13)) {g=true}
  n15 = TextureSampleParameter2D(UVs=ShadingPathSwitch(Default=LinearInterpolate(A=@n12, B=@n13, Alpha=@n14) {desc=Blend bottom}, Mobile=TextureCoordinate() {desc=No hotdogs for mobile})) {parameterName=RadialGradient, texture=/Game/Global/Textures/Default/Masks_and_Gradients/T_New_Soft_Flair}
  Add(
   A=Multiply(
    A=TextureSampleParameter2D() {parameterName=RadialGradient, texture=&T_New_Soft_Flair}.G
    B=Constant3Vector() {constant=<rgba 0 1 0 0>}
   )
   B=Multiply(A=Constant3Vector() {constant=<rgba 1 0 0 0>}, B=@n15.G)
  )
  Constant2Vector() {desc=For testing, g=40, r=10}
  n16 = DynamicParameter() {paramNames [DesaturationFraction, Param2, Param3, Param4]}
  n17 = ParticleColor()
  n18 = Multiply(A=@n17.RGB, B=@n15.G)
  n19 = TextureCoordinate()
  n20 = ScalarParameter() {parameterName=Rotation_Angle}
  n21 = OneMinus(None=StaticComponentMaskParameter(None=MaterialFunctionCall(UVs=@n19, "Rotation Angle (0-1)"=@n20) {materialFunction=&CustomRotator}.Rotated Values) {defaultG=true, parameterName=MaskChannel})
  n22 = OneMinus(None=Power(Base=@n21, Exp=ScalarParameter() {defaultValue=5, parameterName=Squeez_Top_Power}) {constExponent=5})
  n23 = StaticSwitchParameter(
   True=LinearInterpolate(B=ScalarParameter() {defaultValue=1.5, parameterName=Exposure}, Alpha=Multiply(A=@n21, B=@n22)) {constB=1.5}
   False=Constant() {r=1}
  ) {parameterName=FadeTip}
  n24 = Multiply(
   A=StaticSwitchParameter(True=Desaturation(None=@n18, Fraction=@n16.DesaturationFraction), False=@n18) {parameterName=Use Desaturation}
   B=@n23
  )
  n25 = ScalarParameter() {defaultValue=25, parameterName=DepthFadeDist}
  n26 = Multiply(
   A=@n17.A
   B=MaterialFunctionCall(Distance=ScalarParameter() {defaultValue=256, parameterName=Near Camera Fade}) {materialFunction=&Near_Camera_Fade}.Result
  )
  n27 = Multiply(A=ShadingPathSwitch(Default=DepthFade(Opacity=@n26, FadeDistance=@n25), Forward=@n26, Mobile=@n26), B=@n23)
  n28 = StaticSwitchParameter(True=Multiply(A=@n24, B=@n27), False=@n27) {parameterName=UseTranslucent}
  n29 = ScreenPosition()
  n30 = Constant()
  n31 = MaterialFunctionCall(UVs=TextureCoordinate(), CenterPosition=Constant2Vector() {g=0.5, r=0.5}, Radius=Constant() {r=0.5}) {materialFunction=&RadialGradientExponential}
  n32 = LinearInterpolate(
   B=Multiply(A=AppendVector(A=@n29.ViewportUV, B=@n30), B=OneMinus(None=@n31.RadialGradientExponential))
   Alpha=ScalarParameter() {defaultValue=1, parameterName=UseRadialGradient?}
  ) {constA=1}
  n33 = VectorParameter() {group="3) Colorize Time of Day Tint", parameterName="Color_Morning (V3)", sortPriority=1, defaultValue=<rgba 3 3 3 3>}
  n34 = VectorParameter() {group="3) Colorize Time of Day Tint", parameterName="Color_Day (V3)", sortPriority=2, defaultValue=<rgba 5 5 5 3>}
  n35 = VectorParameter() {group="3) Colorize Time of Day Tint", parameterName="Color_Evening (V3)", sortPriority=3, defaultValue=<rgba 2 2 2 3>}
  n36 = VectorParameter() {group="3) Colorize Time of Day Tint", parameterName="Color_Night (V3)", sortPriority=4, defaultValue=<rgba 0.1 0.1 0.1 0.1>}
  n37 = ScalarParameter() {parameterName=1ForTODBreakout, sliderMax=1}
  LinearInterpolate(
   A=@n28
   B=Multiply(
    A=Multiply(A=@n28, B=@n32)
    B=MaterialFunctionCall(Color_Morning=@n33.A, Color_Day=@n34.A, Color_Evening=@n35.A, Color_Night=@n36.A) {materialFunction=&Color_Time_of_Day}.Result
   )
   Alpha=@n37
  )
  n38 = MaterialFunctionCall("FadeSensitivity (optional)"=ScalarParameter() {defaultValue=1, parameterName=AngleFade}) {materialFunction=&FadeBasedOnViewAngle}
  n39 = Constant() {r=1}
  n40 = VectorParameter() {group="3) Colorize Time of Day Tint", parameterName="Color_Morning (V3)", sortPriority=1, defaultValue=<rgba 3 3 3 3>}
  n41 = VectorParameter() {group="3) Colorize Time of Day Tint", parameterName="Color_Day (V3)", sortPriority=2, defaultValue=<rgba 5 5 5 3>}
  n42 = VectorParameter() {group="3) Colorize Time of Day Tint", parameterName="Color_Evening (V3)", sortPriority=3, defaultValue=<rgba 2 2 2 3>}
  n43 = VectorParameter() {group="3) Colorize Time of Day Tint", parameterName="Color_Night (V3)", sortPriority=4, defaultValue=<rgba 0.1 0.1 0.1 0.1>}
  n44 = Multiply(
   A=MaterialFunctionCall(Color_Morning=@n40.RGB, Color_Day=@n41.RGB, Color_Evening=@n42.RGB, Color_Night=@n43.RGB) {materialFunction=&Color_Time_of_Day}.Result
   B=@n24
  )
  MaterialFunctionCall(Emissive=Multiply(A=StaticSwitchParameter(True=@n38.Result, False=@n39) {parameterName=Use Angle Fade}, B=LinearInterpolate(A=@n24, B=@n44, Alpha=@n37))) {materialFunction=&MF_FX_EmissiveAdjustments}
  Round(None=@n11)
  Round(None=@n14)
 ]
}
