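# Niagara system for the idle effects of the Realm Male glider prop: wingtip contrails, engine glows, lights and ribbons, and blinking lights, all attached to User.GliderMesh.
# Contrails spawn by distance when deployed and moving forward, or by rate in front-end preview.
# Two system Timelines ramp engine brightness and ribbon shape, with ribbons bent by local owner velocity; preview overrides most ramps.
# Far-away bloom and light layers set large near-camera fade distances.
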NiagaraSystem /Game/Gadgets/Assets/VinderTech_GliderChute/Glider_Realm_Male/FX_Cinematic/NS_Prop_Realm_Male_Idle
dependencies {
 usedDataInterfaces [&NiagaraDataInterfaceCurve]
 usedDynamicInputs []
 usedModules [&SystemState, &Timeline]
}
emitters {
 AirFoil_Contrails {
  emitterUpdateScript [
   &EmitterState()
   &SpawnPerUnit(
    Module.Use Max Movement Threshold=false
    Module.Use Movement Tolerance=false
    Spawn Probability=1
    Spawn Spacing:NiagaraFloat=&Multiply_Float(
     A=1
     B=&MakeCustomFloatFromBool(
      Bool=&Boolean_AndOperation(
       Bool1=&Boolean_AndOperation(Bool1=$User.bIsFullyDeployed, Bool2=&SetBoolByFloatComparison(A=$User.ForwardDot, B=0.1))
       Bool2=true
      )
      True Float=1
      False Float=0
     )
    )
    Use Spawn Probability=true
   )
   &SpawnRate(
    SpawnRate=&MakeCustomFloatFromBool(Bool=$User.bisFrontEndPreview, True Float=60, False Float=0)
    Use Spawn Probability=false
   )
  ]
  particleSpawnScript [
   &FN_SetSpawnGroup(Inclusivity=true, Normalize Spawn Group=true, Number of Spawn Groups=2, Particles Per Group=15)
   &InitializeRibbon(
    Lifetime=0.333
    Ribbon Width=&Multiply_Float(A=$User.MeshScale, B=10)
    UsePositionOffset=false
    Write Material Random=true
   )
   /SetVariables(
    Particles.RibbonID=/Niagara/DynamicInputs/NiagaraID/MakeNiagaraIDFromInt(Index:NiagaraInt32=$Particles.SpawnGroup, Acquire Tag:NiagaraInt32=0)
    Particles.RibbonLinkOrder:NiagaraFloat=$Particles.Normalized Spawn Group
   ) {isSetParameters=true}
   &SkeletalMeshLocation(
    Bone / Socket Index=&Add_Integer(A=$Particles.SpawnGroup, B=2)
    MaskSpawnGroup=false
    Mesh Sampling Type=::NewEnumerator5
    Skeletal Mesh=$User.GliderMesh
    Socket Sampling Mode=::NewEnumerator5
   )
   &AddVelocity(Velocity=&MakeCustomVectorFromBool(Bool=$User.bisFrontEndPreview, True Vector=<0 1000 0>, False Vector=<0 0 0>))
  ]
  particleUpdateScript [
   &SolveForcesAndVelocity(Clamp Velocity=false, Limit Acceleration=false)
   &ParticleState()
   &ScaleColor(
    Scale Alpha=&FloatFromCurve(FloatCurve=DataInterface() {propertyValues {Curve {keys [<at 0 1>,<at 1 0>]}}})
    ScaleA=true
    ScaleRGB=true
   )
  ]
  rendererClasses [&NiagaraRibbonRendererProperties]
  renderers [{rendererClass=&NiagaraRibbonRendererProperties}]
 }
 Engine_Covers {
  emitterUpdateScript [
   &EmitterState(Life Cycle Mode=::NewEnumerator1, Loop Behavior=::NewEnumerator1, Loop Duration Mode=::NewEnumerator1, UseLoopDelay=false)
   &SpawnBurst_Instantaneous(Use Spawn Probability=false, UseLoopCountLimit=false)
   &SpawnBurst_Instantaneous(Spawn Group=1, Use Spawn Probability=false, UseLoopCountLimit=false) {module=SpawnBurst_Instantaneous001}
  ]
  particleSpawnScript [
   &InitializeParticle(
    Color=<rgba 0.275 0.5087 1.1 1>
    Lifetime=5
    Position Mode=::NewEnumerator0
    Sprite Size=&Multiply_Vector2D_ByFloat(Float=$User.MeshScale, Vector2D=<75 10>)
    Sprite Size Mode=::NewEnumerator4
    Write Material Random=true
   )
   &SpriteFacingAndAlignment(Sprite Alignment=<0 0 1>, Sprite Facing=<0 1 0>, WriteAlignment=true, WriteFacing=true)
  ]
  particleUpdateScript [
   &ParticleState(Kill Particles When Lifetime Has Elapsed=false, Let Infinitely Lived Particles Die When Emitter Deactivates=true)
   &SkeletalMeshLocation(
    Bone / Socket Index=&SpawnGroup()
    MaskSpawnGroup=false
    Mesh Sampling Type=::NewEnumerator5
    Orientation Sampling=::NewEnumerator1
    Skeletal Mesh=$User.GliderMesh
    Socket Sampling Mode=::NewEnumerator5
   )
   &ScaleColor(
    Scale Alpha=&Lerp_Float(
     A=0.0125
     B=1
     Alpha=&MakeCustomFloatFromBool(
      Bool=$User.bisFrontEndPreview
      True Float=1
      False Float=&ClampFloat(Float=&Add_Float(A=$System.Timeline001.NormalizedTime, B=$System.Timeline.NormalizedTime))
     )
    )
    Scale RGB=&VectorFromFloat(Value=&Lerp_Float(
      A=1
      B=1.5
      Alpha=&MakeCustomFloatFromBool(
       Bool=$User.bisFrontEndPreview
       True Float=1
       False Float=&ClampFloat(Float=&Add_Float(A=$System.Timeline001.NormalizedTime, B=$System.Timeline.NormalizedTime))
      )
     ))
    ScaleA=true
    ScaleRGB=true
   )
   /SetVariables(
    Particles.Light Position 2:NiagaraPosition=&Add_Vector(A=Linked() {linkedVariable {name=Particles.Position, type=NiagaraPosition}}, B=<-15 25 -1>)
    Particles.LightColor=&MultiplyLinearColor(
     Linear Color=<rgba 0.275 0.5087 1.1 1>
     Scale Factor=&MakeVector4_FromFloat(Float=&Multiply_Float(
       A=50
       B=&MakeCustomFloatFromBool(
        Bool=$User.bisFrontEndPreview
        True Float=10
        False Float=&ClampFloat(Float=&Add_Float(A=$System.Timeline001.NormalizedTime, B=$System.Timeline.NormalizedTime), Min=0.25)
       )
      ))
    )
    Particles.LightExponent=2
    Particles.LightPosition:NiagaraPosition=&Add_Vector(A=Linked() {linkedVariable {name=Particles.Position, type=NiagaraPosition}}, B=<15 25 -1>)
    Particles.LightRadius=25
   ) {isSetParameters=true, module=SetVariables_#1}
  ]
  rendererClasses [&NiagaraSpriteRendererProperties, &NiagaraLightRendererProperties]
  renderers [
   {rendererClass=&NiagaraSpriteRendererProperties}
   {rendererClass=&NiagaraLightRendererProperties, rendererIndex=1}
   {rendererClass=&NiagaraLightRendererProperties, rendererIndex=2}
  ]
 }
 Engine_FarAwayBloom {
  emitterUpdateScript [
   &EmitterState(Life Cycle Mode=::NewEnumerator1, Loop Behavior=::NewEnumerator1, Loop Duration Mode=::NewEnumerator1, UseLoopDelay=false)
   &SpawnBurst_Instantaneous(Use Spawn Probability=false, UseLoopCountLimit=false)
   &SpawnBurst_Instantaneous(Spawn Group=1, Use Spawn Probability=false, UseLoopCountLimit=false) {module=SpawnBurst_Instantaneous001}
   /SetVariables(Emitter.Emitter_Near_Cam_Fade=5000) {isSetParameters=true}
  ]
  particleSpawnScript [
   &InitializeParticle(
    Color=<rgba 0 1.4167 5 1>
    Lifetime=5
    Position Mode=::NewEnumerator0
    Sprite Size=&Multiply_Vector2D_ByFloat(Float=5, Vector2D=&Multiply_Vector2D_ByFloat(Float=$User.MeshScale, Vector2D=<45 20>))
    Sprite Size Mode=::NewEnumerator4
    Write Material Random=true
   )
   &SpriteFacingAndAlignment(Sprite Alignment=<0 0 1>, Sprite Facing=<0 1 0>, WriteAlignment=true, WriteFacing=true)
   /SetVariables(
    Particles.Color=<rgba 0 2.8333 10 1>
    Particles.ColorAlt:LinearColor=&MultiplyLinearColor(
     Linear Color=$Particles.Color
     Scale Factor=&MakeFloat4FromVectorAndFloat("Vector (RGB)"=&VectorFromFloat(Value=4), "Float (Alpha)"=0.02)
    )
    Particles.SpriteSize=&Multiply_Vector2D_ByFloat(Float=5, Vector2D=&Multiply_Vector2D_ByFloat(Float=$User.MeshScale, Vector2D=<20 10>))
    Particles.SpriteSizeAlt:Vector2f=&Multiply_Vector2D(A=$Particles.SpriteSize, B=<1.5 2>)
   ) {isSetParameters=true, module=SetVariables_#2}
  ]
  particleUpdateScript [
   &ParticleState(Kill Particles When Lifetime Has Elapsed=false, Let Infinitely Lived Particles Die When Emitter Deactivates=true)
   &SkeletalMeshLocation(
    Bone / Socket Index=&SpawnGroup()
    MaskSpawnGroup=false
    Mesh Sampling Type=::NewEnumerator5
    Orientation Sampling=::NewEnumerator1
    Sampled Position Offset=<0 18 0>
    Skeletal Mesh=$User.GliderMesh
    Socket Sampling Mode=::NewEnumerator5
   )
   &ScaleColor(
    Scale Alpha=&Lerp_Float(
     A=0.0125
     B=1
     Alpha=&MakeCustomFloatFromBool(
      Bool=$User.bisFrontEndPreview
      True Float=1
      False Float=&ClampFloat(Float=&Add_Float(A=$System.Timeline001.NormalizedTime, B=$System.Timeline.NormalizedTime))
     )
    )
    Scale RGB=&VectorFromFloat(Value=&Lerp_Float(
      A=1
      B=1.5
      Alpha=&MakeCustomFloatFromBool(
       Bool=$User.bisFrontEndPreview
       True Float=1
       False Float=&ClampFloat(Float=&Add_Float(A=$System.Timeline001.NormalizedTime, B=$System.Timeline.NormalizedTime))
      )
     ))
    ScaleA=true
    ScaleRGB=true
   )
   /SetVariables(
    Particles.Light Position 2:NiagaraPosition=&Add_Vector(A=Linked() {linkedVariable {name=Particles.Position, type=NiagaraPosition}}, B=<-15 25 -1>)
    Particles.LightColor=&MultiplyLinearColor(
     Linear Color=<rgba 0.275 0.5087 1.1 1>
     Scale Factor=&MakeVector4_FromFloat(Float=&Multiply_Float(
       A=50
       B=&MakeCustomFloatFromBool(
        Bool=$User.bisFrontEndPreview
        True Float=10
        False Float=&ClampFloat(Float=&Add_Float(A=$System.Timeline001.NormalizedTime, B=$System.Timeline.NormalizedTime), Min=0.25)
       )
      ))
    )
    Particles.LightExponent=2
    Particles.LightPosition:NiagaraPosition=&Add_Vector(A=Linked() {linkedVariable {name=Particles.Position, type=NiagaraPosition}}, B=<15 25 -1>)
    Particles.LightRadius=25
   ) {isSetParameters=true, module=SetVariables_#1}
   &CameraOffset(Camera Offset Amount=10)
   &DynamicMaterialParameters(Module.Param0WriteEnabled=true, Module.Param1WriteEnabled=true, Module.Param2WriteEnabled=true, Module.Param3WriteEnabled=true)
  ]
  rendererClasses [&NiagaraLightRendererProperties, &NiagaraSpriteRendererProperties]
  renderers [
   {rendererClass=&NiagaraLightRendererProperties, rendererIndex=0}
   {rendererClass=&NiagaraLightRendererProperties, rendererIndex=1}
   {rendererClass=&NiagaraSpriteRendererProperties, rendererIndex=2}
   {rendererClass=&NiagaraSpriteRendererProperties, rendererIndex=3}
  ]
 }
 Engine_FarAwayBloom001 {
  emitterUpdateScript [
   &EmitterState(Life Cycle Mode=::NewEnumerator1, Loop Behavior=::NewEnumerator1, Loop Duration Mode=::NewEnumerator1, UseLoopDelay=false)
   &SpawnBurst_Instantaneous(Use Spawn Probability=false, UseLoopCountLimit=false)
   &SpawnBurst_Instantaneous(Spawn Group=1, Use Spawn Probability=false, UseLoopCountLimit=false) {module=SpawnBurst_Instantaneous001}
   /SetVariables(Emitter.Emitter_Near_Cam_Fade=50000) {isSetParameters=true}
  ]
  particleSpawnScript [
   &InitializeParticle(
    Color=<rgba 0 2.8333 10 1>
    Lifetime=5
    Position Mode=::NewEnumerator0
    Sprite Size=&Multiply_Vector2D_ByFloat(Float=1, Vector2D=&Multiply_Vector2D_ByFloat(Float=$User.MeshScale, Vector2D=<45 20>))
    Sprite Size Mode=::NewEnumerator4
    Write Material Random=true
   )
   &SpriteFacingAndAlignment(Sprite Alignment=<0 0 1>, Sprite Facing=<0 1 0>, WriteAlignment=true, WriteFacing=true)
  ]
  particleUpdateScript [
   &ParticleState(Kill Particles When Lifetime Has Elapsed=false, Let Infinitely Lived Particles Die When Emitter Deactivates=true)
   &SkeletalMeshLocation(
    Bone / Socket Index=&SpawnGroup()
    MaskSpawnGroup=false
    Mesh Sampling Type=::NewEnumerator5
    Orientation Sampling=::NewEnumerator1
    Sampled Position Offset=<0 18 0>
    Skeletal Mesh=$User.GliderMesh
    Socket Sampling Mode=::NewEnumerator5
   )
   &ScaleColor(
    Scale Alpha=&Lerp_Float(
     A=0.0125
     B=1
     Alpha=&MakeCustomFloatFromBool(
      Bool=$User.bisFrontEndPreview
      True Float=1
      False Float=&ClampFloat(Float=&Add_Float(A=$System.Timeline001.NormalizedTime, B=$System.Timeline.NormalizedTime))
     )
    )
    Scale RGB=&VectorFromFloat(Value=&Lerp_Float(
      A=1
      B=1.5
      Alpha=&MakeCustomFloatFromBool(
       Bool=$User.bisFrontEndPreview
       True Float=1
       False Float=&ClampFloat(Float=&Add_Float(A=$System.Timeline001.NormalizedTime, B=$System.Timeline.NormalizedTime))
      )
     ))
    ScaleA=true
    ScaleRGB=true
   )
   /SetVariables(
    Particles.Color=<rgba 0 2.8333 10 1>
    Particles.ColorAlt:LinearColor=&MultiplyLinearColor(
     Linear Color=$Particles.Color
     Scale Factor=&MakeFloat4FromVectorAndFloat("Vector (RGB)"=&VectorFromFloat(Value=2), "Float (Alpha)"=0.04)
    )
    Particles.SpriteSize=&Multiply_Vector2D_ByFloat(Float=10, Vector2D=&Multiply_Vector2D_ByFloat(Float=$User.MeshScale, Vector2D=<20 10>))
    Particles.SpriteSizeAlt:Vector2f=&Multiply_Vector2D(A=$Particles.SpriteSize, B=<3 5>)
   ) {isSetParameters=true, module=SetVariables_#2}
   &CameraOffset(Camera Offset Amount=10)
   &DynamicMaterialParameters(Module.Param0WriteEnabled=true, Module.Param1WriteEnabled=true, Module.Param2WriteEnabled=true, Module.Param3WriteEnabled=true)
  ]
  rendererClasses [&NiagaraSpriteRendererProperties]
  renderers [
   {rendererClass=&NiagaraSpriteRendererProperties}
   {rendererClass=&NiagaraSpriteRendererProperties, rendererIndex=1}
  ]
 }
 Engine_Ribbon_Left {
  emitterUpdateScript [
   &EmitterState(Life Cycle Mode=::NewEnumerator1, Loop Behavior=::NewEnumerator1, Loop Duration Mode=::NewEnumerator1, UseLoopDelay=false)
   &SpawnBurst_Instantaneous(Spawn Count=10, Use Spawn Probability=false, UseLoopCountLimit=false)
  ]
  particleSpawnScript [
   &InitializeParticle(Color Mode=::NewEnumerator0, Lifetime=0, UsePositionOffset=false, Write Material Random=true)
   &InitializeRibbon(
    Color=<rgba 0.275 0.5087 1.1 1>
    Ribbon Width=&Multiply_Float(A=$User.MeshScale, B=1.5)
    UsePositionOffset=false
    Write Material Random=true
   )
   /SetVariables(
    Particles.RibbonID=/Niagara/DynamicInputs/Meshes/MakeRibbonIDFromSpawnGroup()
    Particles.RibbonLinkOrder:NiagaraFloat=&ReturnNormalizedExecIndex()
   ) {isSetParameters=true, module=SetVariables_#3}
  ]
  particleUpdateScript [
   &ParticleState(Kill Particles When Lifetime Has Elapsed=false, Let Infinitely Lived Particles Die When Emitter Deactivates=true)
   &SkeletalMeshLocation(
    MaskSpawnGroup=false
    Mesh Sampling Type=::NewEnumerator5
    Sampled Orientation Coordinate Space=::Local
    Skeletal Mesh=$User.GliderMesh
    Socket Sampling Mode=::NewEnumerator5
   )
   /SetVariables(Particles.Position=/Niagara/DynamicInputs/Lerp/Lerp_Position(
     A:NiagaraPosition=$Particles.Position
     B:NiagaraPosition=&Add_Vector(
      A=Linked() {linkedVariable {name=Particles.Position, type=NiagaraPosition}}
      B=&MakeVector(
       X=&Lerp_Float(
        A=0
        B=&Lerp_Float(
         A=0
         B=&Multiply_Float(
          A=&MakeFloatFromVector(
           Vector=&NormalizeVector(
            Vector To Normalize=&TransformVector(Vector=$Engine.Owner.Velocity, Source Space=::World, Destination Space=::Local)
            Safe Normalize=true
            Safe Fallback Threshold=0
            Safe Fallback Vector=<0 0 0.001>
           )
           Channel=::NewEnumerator0
          )
          B=-25
         )
         Alpha=&FloatFromCurve(
          FloatCurve=DataInterface() {propertyValues {Curve {keys [
              {interpMode=RCIM_Cubic, leaveTangentWeight=0.3333, tangentMode=RCTM_User, time=0, value=0}
              {arriveTangentWeight=1.2598, interpMode=RCIM_Cubic, tangentMode=RCTM_User, time=1, value=1.25}
             ]}}}
          CurveIndex=&ReturnNormalizedExecIndex()
         )
        )
        Alpha=&ClampFloat(Float=&Add_Float(A=$System.Timeline001.NormalizedTime, B=$System.Timeline.NormalizedTime))
       )
       Y=&Lerp_Float(
        A=&Lerp_Float(
         A=60
         B=90
         Alpha=&MakeCustomFloatFromBool(Bool=$User.bisFrontEndPreview, True Float=1, False Float=&ClampFloat(Float=$System.Timeline001.NormalizedTime))
        )
        B=110
        Alpha=&ClampFloat(Float=$System.Timeline.NormalizedTime)
       )
       Z=0
      )
     )
     Alpha:NiagaraFloat=&ReturnNormalizedExecIndex()
    )) {isSetParameters=true, module=SetVariables_#4}
   &DynamicMaterialParameters(
    Index 0 Param 1=&Lerp_Float(
     A=1
     B=-0.25
     Alpha=&MakeCustomFloatFromBool(
      Bool=$User.bisFrontEndPreview
      True Float=1.15
      False Float=&ClampFloat(Float=&Add_Float(A=$System.Timeline001.NormalizedTime, B=$System.Timeline.NormalizedTime))
     )
    )
    Module.Param0WriteEnabled=true
    Module.Param1WriteEnabled=true
    Module.Param2WriteEnabled=true
    Module.Param3WriteEnabled=true
   )
   &ScaleRibbonWidth(Ribbon Width Scale=&Lerp_Float(
     A=1
     B=&FloatFromCurve(
      FloatCurve=DataInterface() {propertyValues {Curve {keys [
          {interpMode=RCIM_Cubic, leaveTangentWeight=0.3333, tangentMode=RCTM_User, time=0.25, value=1}
          {arriveTangent=0.5619, arriveTangentWeight=0.3333, interpMode=RCIM_Cubic, leaveTangent=0.5619, tangentMode=RCTM_User, time=1, value=1.1}
         ]}}}
      CurveIndex=&ReturnNormalizedExecIndex()
     )
     Alpha=&ClampFloat(Float=&Add_Float(A=$System.Timeline001.NormalizedTime, B=$System.Timeline.NormalizedTime))
    ))
   &ScaleColor(
    Scale RGB=&VectorFromFloat(Value=&Lerp_Float(
      A=1
      B=1.5
      Alpha=&MakeCustomFloatFromBool(
       Bool=$User.bisFrontEndPreview
       True Float=1
       False Float=&ClampFloat(Float=&Add_Float(A=$System.Timeline001.NormalizedTime, B=$System.Timeline.NormalizedTime))
      )
     ))
    ScaleA=true
    ScaleRGB=true
   )
  ]
  rendererClasses [&NiagaraRibbonRendererProperties]
  renderers [{rendererClass=&NiagaraRibbonRendererProperties}]
 }
 Engine_Ribbon_Right {
  emitterUpdateScript [
   &EmitterState(Life Cycle Mode=::NewEnumerator1, Loop Behavior=::NewEnumerator1, Loop Duration Mode=::NewEnumerator1, UseLoopDelay=false)
   &SpawnBurst_Instantaneous(Spawn Count=10, Use Spawn Probability=false, UseLoopCountLimit=false)
  ]
  particleSpawnScript [
   &InitializeParticle(Color Mode=::NewEnumerator0, Lifetime=5, UsePositionOffset=false, Write Material Random=true)
   &InitializeRibbon(
    Color=<rgba 0.275 0.5087 1.1 1>
    Ribbon Width=&Multiply_Float(A=$User.MeshScale, B=1.5)
    UsePositionOffset=false
    Write Material Random=true
   )
   /SetVariables(
    Particles.RibbonID=&MakeRibbonIDFromSpawnGroup()
    Particles.RibbonLinkOrder:NiagaraFloat=&ReturnNormalizedExecIndex()
   ) {isSetParameters=true, module=SetVariables_#3}
  ]
  particleUpdateScript [
   &ParticleState(Kill Particles When Lifetime Has Elapsed=false, Let Infinitely Lived Particles Die When Emitter Deactivates=true)
   &SkeletalMeshLocation(
    Bone / Socket Index=1
    MaskSpawnGroup=false
    Mesh Sampling Type=::NewEnumerator5
    Orientation Sampling=::NewEnumerator1
    Position Sampling=::NewEnumerator2
    Sampled Orientation Coordinate Space=::Local
    Skeletal Mesh=$User.GliderMesh
    Socket Sampling Mode=::NewEnumerator5
   )
   /SetVariables(Particles.Position=&Lerp_Position(
     A:NiagaraPosition=$Particles.SkeletalMeshLocation.SampledPosition
     B:NiagaraPosition=&Add_Vector(
      A=Linked() {linkedVariable {name=Particles.SkeletalMeshLocation.SampledPosition, type=NiagaraPosition}}
      B=&MakeVector(
       X=&Lerp_Float(
        A=0
        B=&Lerp_Float(
         A=0
         B=&Multiply_Float(
          A=&MakeFloatFromVector(
           Vector=&NormalizeVector(
            Vector To Normalize=&TransformVector(Vector=$Engine.Owner.Velocity, Source Space=::World, Destination Space=::Local)
            Safe Normalize=true
            Safe Fallback Threshold=0
            Safe Fallback Vector=<0 0 0.001>
           )
           Channel=::NewEnumerator0
          )
          B=-25
         )
         Alpha=&FloatFromCurve(
          FloatCurve=DataInterface() {propertyValues {Curve {keys [
              {interpMode=RCIM_Cubic, leaveTangentWeight=0.3333, tangentMode=RCTM_User, time=0, value=0}
              {arriveTangentWeight=1.2598, interpMode=RCIM_Cubic, tangentMode=RCTM_User, time=1, value=1.25}
             ]}}}
          CurveIndex=&ReturnNormalizedExecIndex()
         )
        )
        Alpha=&ClampFloat(Float=&Add_Float(A=$System.Timeline001.NormalizedTime, B=$System.Timeline.NormalizedTime))
       )
       Y=&Lerp_Float(
        A=&Lerp_Float(
         A=60
         B=90
         Alpha=&MakeCustomFloatFromBool(Bool=$User.bisFrontEndPreview, True Float=1, False Float=&ClampFloat(Float=$System.Timeline001.NormalizedTime))
        )
        B=110
        Alpha=&ClampFloat(Float=$System.Timeline.NormalizedTime)
       )
       Z=0
      )
     )
     Alpha:NiagaraFloat=&ReturnNormalizedExecIndex()
    )) {isSetParameters=true, module=SetVariables_#4}
   &DynamicMaterialParameters(
    Index 0 Param 1=&Lerp_Float(
     A=1
     B=-0.25
     Alpha=&MakeCustomFloatFromBool(
      Bool=$User.bisFrontEndPreview
      True Float=1.15
      False Float=&ClampFloat(Float=&Add_Float(A=$System.Timeline001.NormalizedTime, B=$System.Timeline.NormalizedTime))
     )
    )
    Module.Param0WriteEnabled=true
    Module.Param1WriteEnabled=true
    Module.Param2WriteEnabled=true
    Module.Param3WriteEnabled=true
   )
   &ScaleRibbonWidth(Ribbon Width Scale=&Lerp_Float(
     A=1
     B=&FloatFromCurve(
      FloatCurve=DataInterface() {propertyValues {Curve {keys [
          {interpMode=RCIM_Cubic, leaveTangentWeight=0.3333, tangentMode=RCTM_User, time=0.25, value=1}
          {arriveTangent=0.5619, arriveTangentWeight=0.3333, interpMode=RCIM_Cubic, leaveTangent=0.5619, tangentMode=RCTM_User, time=1, value=1.1}
         ]}}}
      CurveIndex=&ReturnNormalizedExecIndex()
     )
     Alpha=&ClampFloat(Float=&Add_Float(A=$System.Timeline001.NormalizedTime, B=$System.Timeline.NormalizedTime))
    ))
   &ScaleColor(
    Scale RGB=&VectorFromFloat(Value=&Lerp_Float(
      A=1
      B=1.5
      Alpha=&MakeCustomFloatFromBool(
       Bool=$User.bisFrontEndPreview
       True Float=1
       False Float=&ClampFloat(Float=&Add_Float(A=$System.Timeline001.NormalizedTime, B=$System.Timeline.NormalizedTime))
      )
     ))
    ScaleA=true
    ScaleRGB=true
   )
  ]
  rendererClasses [&NiagaraRibbonRendererProperties]
  renderers [{rendererClass=&NiagaraRibbonRendererProperties}]
 }
 Flashing_Lights {
  emitterUpdateScript [
   &EmitterState(Life Cycle Mode=::NewEnumerator1, Loop Behavior=::NewEnumerator1, Loop Duration Mode=::NewEnumerator1, UseLoopDelay=false)
   &SpawnBurst_Instantaneous(Spawn Count=6, Use Spawn Probability=false, UseLoopCountLimit=false)
  ]
  particleSpawnScript [
   &FN_SetSpawnGroup(Number of Spawn Groups=6)
   &InitializeParticle(
    Color=<rgba 20 10.136 0 1>
    Sprite Rotation Mode=::NewEnumerator2
    Sprite Size Mode=::NewEnumerator3
    Uniform Sprite Size=&Multiply_Float(A=$User.MeshScale, B=20)
    UsePositionOffset=false
    Write Material Random=true
   )
   &CameraOffset(Camera Offset Amount=5)
  ]
  particleUpdateScript [
   &ParticleState(Kill Particles When Lifetime Has Elapsed=false, Let Infinitely Lived Particles Die When Emitter Deactivates=true)
   &SkeletalMeshLocation(
    Bone / Socket Index=&Add_Integer(A=$Particles.SpawnGroup, B=2)
    MaskSpawnGroup=false
    Mesh Sampling Type=::NewEnumerator5
    Skeletal Mesh=$User.GliderMesh
    Socket Sampling Mode=::NewEnumerator5
   )
   &ScaleColor(
    Scale Alpha=&MakeCustomFloatFromBool(Bool=$User.Blink, True Float=1, False Float=0)
    Scale RGB=&VectorFromFloat(Value=&Waveform(Global Clamp Min=0.6, "Waveform [1]"=::NewEnumerator0, "[1] Frequency"=1, "[1] Amplitude Min/Max"=<0 1>))
    ScaleA=true
    ScaleRGB=true
   )
   &DynamicMaterialParameters(
    Index 0 Param 1=1
    Module.Param0WriteEnabled=true
    Module.Param1WriteEnabled=true
    Module.Param2WriteEnabled=true
    Module.Param3WriteEnabled=true
   )
   &ScaleSpriteSize(
    Scale Sprite Size Mode=::NewEnumerator0
    Uniform Scale Factor=&Waveform(Global Clamp Min=0.8, "Waveform [1]"=::NewEnumerator0, "[1] Frequency"=1, "[1] Amplitude Min/Max"=<0 1>)
   )
  ]
  rendererClasses [&NiagaraSpriteRendererProperties]
  renderers [
   {rendererClass=&NiagaraSpriteRendererProperties}
   {rendererClass=&NiagaraSpriteRendererProperties, rendererIndex=1}
  ]
 }
 Flashing_Lights_FarAway {
  emitterUpdateScript [
   &EmitterState(Life Cycle Mode=::NewEnumerator1, Loop Behavior=::NewEnumerator1, Loop Duration Mode=::NewEnumerator1, UseLoopDelay=false)
   &SpawnBurst_Instantaneous(Spawn Count=6, Use Spawn Probability=false, UseLoopCountLimit=false)
   /SetVariables(Emitter.Emitter_Near_Cam_Fade=80000) {isSetParameters=true}
  ]
  particleSpawnScript [
   &FN_SetSpawnGroup(Number of Spawn Groups=6)
   &InitializeParticle(
    Color=<rgba 1.5 0.9821 0.45 1>
    Sprite Rotation Mode=::NewEnumerator2
    Sprite Size Mode=::NewEnumerator3
    Uniform Sprite Size=&Multiply_Float(A=$User.MeshScale, B=60)
    UsePositionOffset=false
    Write Material Random=true
   )
   &CameraOffset(Camera Offset Amount=5)
  ]
  particleUpdateScript [
   &ParticleState(Kill Particles When Lifetime Has Elapsed=false, Let Infinitely Lived Particles Die When Emitter Deactivates=true)
   &SkeletalMeshLocation(
    Bone / Socket Index=&Add_Integer(A=$Particles.SpawnGroup, B=2)
    MaskSpawnGroup=false
    Mesh Sampling Type=::NewEnumerator5
    Skeletal Mesh=$User.GliderMesh
    Socket Sampling Mode=::NewEnumerator5
   )
   &ScaleColor(
    Scale Alpha=&MakeCustomFloatFromBool(Bool=$User.Blink, True Float=1, False Float=0)
    Scale RGB=&VectorFromFloat(Value=&Waveform(Global Clamp Min=0.6, "Waveform [1]"=::NewEnumerator0, "[1] Frequency"=1, "[1] Amplitude Min/Max"=<0 1>))
    ScaleA=true
    ScaleRGB=true
   )
   &DynamicMaterialParameters(
    Index 0 Param 1=1
    Module.Param0WriteEnabled=true
    Module.Param1WriteEnabled=true
    Module.Param2WriteEnabled=true
    Module.Param3WriteEnabled=true
   )
   &ScaleSpriteSize(
    Scale Sprite Size Mode=::NewEnumerator0
    Uniform Scale Factor=&Waveform(Global Clamp Min=0.8, "Waveform [1]"=::NewEnumerator0, "[1] Frequency"=1, "[1] Amplitude Min/Max"=<0 1>)
   )
  ]
  rendererClasses [&NiagaraSpriteRendererProperties]
  renderers [
   {rendererClass=&NiagaraSpriteRendererProperties}
   {rendererClass=&NiagaraSpriteRendererProperties, rendererIndex=1}
  ]
 }
}
summary {
 systemName=NS_Prop_Realm_Male_Idle
 emitters [
  {emitterName=AirFoil_Contrails, rendererClasses [&NiagaraRibbonRendererProperties]}
  {emitterName=Flashing_Lights, rendererClasses [&NiagaraSpriteRendererProperties]}
  {emitterName=Engine_Ribbon_Left, rendererClasses [&NiagaraRibbonRendererProperties]}
  {emitterName=Engine_Covers, rendererClasses [&NiagaraSpriteRendererProperties, &NiagaraLightRendererProperties]}
  {emitterName=Engine_Ribbon_Right, rendererClasses [&NiagaraRibbonRendererProperties]}
  {
   emitterName=Engine_FarAwayBloom
   rendererClasses [&NiagaraLightRendererProperties, &NiagaraSpriteRendererProperties]
  }
  {emitterName=Engine_FarAwayBloom001, rendererClasses [&NiagaraSpriteRendererProperties]}
  {emitterName=Flashing_Lights_FarAway, rendererClasses [&NiagaraSpriteRendererProperties]}
 ]
 userVariables [
  User.bisFrontEndPreview {type=NiagaraBool}
  User.bIsFullyDeployed {type=NiagaraBool}
  User.Blink {type=NiagaraBool, defaultValue=true}
  User.ForwardDot {type=NiagaraFloat, defaultValue=0, description=""}
  User.GliderMesh {
   type=NiagaraDataInterfaceSkeletalMesh
   defaultValue {
    struct=NiagaraExt_VariableValue_DataInterface
    value {dataInterface=&self:NiagaraDataInterfaceSkeletalMesh_0}
   }
  }
  User.MeshScale {type=NiagaraFloat, defaultValue=3, description=""}
 ]
}
